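# Unlit translucent material for the SlickSwish backpack's main lightning effect, built for Niagara mesh particles.
# Emissive blends tinted triplanar nebula with noise-modulated curve-atlas colors, scaled by particle color.
# Opacity combines a panning lightning mask, particle-driven dissolve, end tapering, depth fade and view-direction fade.
# StepPanner time mixes smooth and stepped motion, while SplineThicken sets particle-controlled tapered width and a separate sine chain builds wobble displacement.
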Material /BRCosmetics/Accessories/FORT_Backpacks/Backpack_M_MED_SlickSwish/FX/Materials/M_Backpack_SlickSwish_MainLightning
graph {
 properties {
  bAutomaticallySetUsageInEditor=false
  bUsedWithNiagaraMeshParticles=true
  blendMode=BLEND_Translucent
  shadingModel=MSM_Unlit
  usedShadingModels=Unlit
 }
 nodes [
  n1 = TextureObject() {texture=/Game/Characters/Player/Female/Medium/Bodies/F_MED_SuperNovaTaro/FX/Material/T_SuperNovaTaro_Skin_Nebula}
  n2 = Constant() {r=3}
  n3 = Multiply(
   A=Power(Base=MaterialFunctionCall!(CameraVector, Texture=@n1, Tiling=@n2, Offset, AxisFadeContrast) {materialFunction=/Engine/Functions/Engine_MaterialFunctions01/Texturing/TriplanarCameraVector}.Result) {constExponent=0.8}
   B=VectorParameter() {parameterName=LerpColor, defaultValue=<rgba 1 0.1645 0.7215 1>}.RGB
  )
  n4 = CurveAtlasRowParameter(CurveTime=ComponentMask(None=TextureCoordinate()) {r=true}) {
   atlas=&/CAtlas_SlickSwish_FX
   curve=&/CV_SlickSwish_Backpack_LightningColor
   parameterName=ColorCurve
  }
  n5 = TextureCoordinate()
  n6 = ComponentMask(None=VectorParameter() {group=ColorRamp, parameterName=ColorRampUVScale, defaultValue=<rgba 1 0.8 0 0>}.RGB) {g=true, r=true}
  n7 = TextureSampleParameter2D(UVs=Panner(Coordinate=Multiply(A=@n5, B=@n6), Time=NamedRerouteUsage()) {speedX=-1}) {
   group=ColorRamp
   parameterName=ColorRampTex
   samplerType=SAMPLERTYPE_LinearGrayscale
   texture=/Engine/Functions/Engine_MaterialFunctions02/ExampleContent/Textures/LowResBlurredNoise
  }
  n8 = CurveAtlasRowParameter(CurveTime=@n7.R) {
   atlas=&CAtlas_SlickSwish_FX
   curve=&/CV_SlickSwish_Backpack_LightningColor2
   defaultValue=1
   parameterName=ColorCurve02
  }
  n9 = ScalarParameter() {defaultValue=1, group=ColorRamp, parameterName=ColorRampPower}
  n10 = ScalarParameter() {defaultValue=0.3, group=Noise, parameterName=NoiseMin}
  n11 = ScalarParameter() {defaultValue=1, group=Noise, parameterName=NoiseMax}
  n12 = TextureCoordinate()
  n13 = ComponentMask(None=VectorParameter() {group=Noise, parameterName=NoiseUVScale, defaultValue=<rgba 1 0.8 0 0>}.RGB) {g=true, r=true}
  n14 = TextureSampleParameter2D(UVs=Panner(Coordinate=Multiply(A=@n12, B=@n13), Time=NamedRerouteUsage()) {speedX=-1}) {
   group=Noise
   parameterName=Noise
   samplerType=SAMPLERTYPE_LinearGrayscale
   texture=&LowResBlurredNoise
  }
  n15 = Multiply(
   A=LinearInterpolate(A=@n4, B=Power(Base=@n8, Exp=@n9), Alpha=ScalarParameter() {group=ColorRamp, parameterName=UseColorRamp})
   B=Power(
    Base=LinearInterpolate(A=@n10, B=@n11, Alpha=@n14.R) {constA=0.3}
    Exp=ScalarParameter() {defaultValue=1.2, group=Noise, parameterName=NoisePower}
   ) {constExponent=1.2}
  )
  n16 = ScalarParameter() {defaultValue=0.5, parameterName=ColorNoiseLerp}
  MaterialFunctionCall(Emissive=Multiply(A=ParticleColor().RGB, B=LinearInterpolate(A=@n3, B=@n15, Alpha=@n16) {constAlpha=0.6})) {materialFunction=&MF_FX_EmissiveAdjustments}
  n17 = Time()
  n18 = Constant() {r=1.5}
  n19 = Ceil(None=Multiply(A=Time() {bOverride_Period=true, period=6000}, B=Constant() {r=8}) {constB=6})
  n20 = Constant() {r=0.5}
  NamedRerouteDeclaration(None=LinearInterpolate(A=Multiply(A=@n17, B=@n18) {constB=1.5}, B=Multiply(A=@n19, B=@n20)) {constAlpha=0.3}) {name=StepPanner, nodeColor=<rgba 0 1 0.8125 1>}
  n21 = VectorParameter() {defaultValue=<rgba 0 1 0 0>}
  n22 = Multiply(A=TextureCoordinate(), B=ScalarParameter() {defaultValue=1, parameterName=Param_3})
  n23 = Multiply(A=Time() {bOverride_Period=true, period=6000}, B=ScalarParameter() {defaultValue=4}) {constB=6}
  n24 = LinearInterpolate(
   A=Multiply(A=Time(), B=ScalarParameter() {defaultValue=1, parameterName=s}) {constB=1.5}
   B=Multiply(A=Ceil(None=@n23), B=ScalarParameter() {defaultValue=1, parameterName=Param_1})
  )
  n25 = Sine(None=Add(A=ComponentMask(None=@n22) {r=true}, B=Multiply(A=@n24) {constB=-1})) {period=0.5}
  n26 = ScalarParameter() {parameterName=565655}
  n27 = Multiply(A=Power(Base=Subtract(A=@n25, B=@n26) {constB=0.5})) {constB=2}
  n28 = TextureCoordinate()
  n29 = Multiply(A=Sine(None=ComponentMask(None=@n28) {r=true}) {period=2})
  n30 = Multiply(A=Multiply(A=Transform(None=@n21.RGB), B=Saturate(None=@n27)), B=Saturate(None=LinearInterpolate(Alpha=@n29) {constA=0.5}))
  n31 = ScalarParameter() {defaultValue=10, parameterName=WPOIntensity}
  Reroute(None=Multiply(A=Multiply(A=@n30, B=@n31), B=DynamicParameter() {defaultValue=<rgba 1 1 0 1>, paramNames [Width,WPOIntensity,Dissolve,Param4]}.WPOIntensity))
  n32 = TextureSampleParameter2D(UVs=Panner(Coordinate=TextureCoordinate() {uTiling=0.8}, Time=NamedRerouteUsage()) {speedX=-0.7}) {
   parameterName=Tex_Mask
   samplerType=SAMPLERTYPE_Grayscale
   texture=&/T_Backpack_SlickSwish_Lightning
  }
  n33 = Power(
   Base=StaticComponentMaskParameter(None=@n32.RGB) {defaultR=true, parameterName=Mask}
   Exp=ScalarParameter() {defaultValue=1, parameterName=MaskPower}
  )
  n34 = DynamicParameter() {defaultValue=<rgba 1 1 0 1>, paramNames [Width,WPOIntensity,Dissolve,Param4]}
  n35 = ComponentMask(None=TextureCoordinate()) {r=true}
  n36 = TextureCoordinate()
  n37 = Power(Base=OneMinus(None=ComponentMask(None=@n36) {r=true})) {constExponent=5}
  n38 = Multiply(A=Multiply(A=Sine(None=@n35) {period=2}) {constB=2}, B=Saturate(None=Multiply(A=@n37) {constB=50}))
  n39 = Multiply(A=ParticleColor().A, B=Multiply(A=Subtract(A=@n33, B=@n34.Dissolve), B=Saturate(None=@n38)))
  n40 = Constant3Vector() {constant=<rgba 1 0 0 0>}
  n41 = Multiply(A=DotProduct(A=CameraVectorWS(), B=Transform(None=@n40) {transformSourceType=TRANSFORMSOURCE_Instance})) {constB=-1}
  n42 = ScalarParameter() {defaultValue=0.15, parameterName=FadeOffset}
  n43 = Saturate(None=Multiply(A=Add(A=@n41, B=@n42), B=ScalarParameter() {defaultValue=20, parameterName=Param_2}) {constB=5})
  n44 = ScalarParameter() {defaultValue=1, parameterName=Param_4}
  Reroute(None=Multiply(A=DepthFade(Opacity=@n39) {fadeDistanceDefault=5}, B=Power(Base=@n43, Exp=@n44)))
  n45 = DynamicParameter() {defaultValue=<rgba 1 1 0 1>, paramNames [Width,WPOIntensity,Dissolve,Param4]}
  n46 = TextureCoordinate()
  n47 = Power(Base=OneMinus(None=ComponentMask(None=@n46) {r=true})) {constExponent=1.5}
  n48 = ComponentMask(None=TextureCoordinate()) {r=true}
  n49 = Multiply(A=Saturate(None=Multiply(A=@n47) {constB=2}), B=Saturate(None=Multiply(A=@n48) {constB=3}))
  n50 = Multiply(A=@n45.Width, B=Multiply(A=Constant() {r=50}, B=@n49))
  n51 = Multiply(A=@n45.Width, B=Multiply(A=@n49, B=Constant() {r=100}))
  n52 = StaticBool()
  Reroute(None=Add(A=MaterialFunctionCall(WidthBase=@n50, WidthTip=@n51, Expand U or V UV Channel=@n52) {materialFunction=&SplineThicken}.WorldPositionOffset) {constB=0})
 ]
}
